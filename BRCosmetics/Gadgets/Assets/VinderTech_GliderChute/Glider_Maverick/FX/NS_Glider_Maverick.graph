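# Niagara effect for the Maverick glider, drawing four ribbon anim trails between bone pairs plus four socket-attached thrusters on the user skeletal mesh.
# Thruster sprites drift along each socket's rotated X axis, while persistent mesh flames with lights pulse in brightness.
# System integrator values fade trail alpha and drive thruster length and intensity.
# Front-end modifiers retune size and brightness for menus, with a timed thruster ramp about three seconds in.
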NiagaraSystem /BRCosmetics/Gadgets/Assets/VinderTech_GliderChute/Glider_Maverick/FX/NS_Glider_Maverick
dependencies {
 usedDataInterfaces [&NiagaraDataInterfaceCurve, &NiagaraDataInterfaceColorCurve]
 usedDynamicInputs []
 usedModules [&SystemState, /Game/Effects/Niagara/Modules/Math/FN_Integrator/FN_Integrator_v2]
}
emitters {
 FN_AnimTrail_V4 {
  emitterUpdateScript [
   &EmitterState()
   &FN_SpawnAnimTrail_v2(Number of Anim Trails=4, Spawn Interval=0)
  ]
  particleSpawnScript [
   /Game/Effects/Niagara/Modules/AnimTrail/FN_InitializeAnimTrail/FN_InitializeAnimTrailBoneIndecies(First Bone or Socket Index=4, Second Bone or Socket Index=5)
   &FN_InitializeAnimTrailBoneIndecies(AnimTrail Selection=1, First Bone or Socket Index=6, Second Bone or Socket Index=7) {module=FN_InitializeAnimTrailBoneIndecies001}
   &FN_InitializeAnimTrailBoneIndecies(AnimTrail Selection=2, First Bone or Socket Index=8, Second Bone or Socket Index=9) {module=FN_InitializeAnimTrailBoneIndecies002}
   &FN_InitializeAnimTrailBoneIndecies(AnimTrail Selection=3, First Bone or Socket Index=10, Second Bone or Socket Index=11) {module=FN_InitializeAnimTrailBoneIndecies003}
   /Game/Effects/Niagara/Modules/AnimTrail/FN_InitializeAnimTrail/FN_InitializeAnimTrail(AnimTrailWidth=1, Color=<rgba 1 1 1 2>, Lifetime=0.15)
  ]
  particleUpdateScript [
   /Game/Effects/Niagara/Modules/AnimTrail/FN_UpdateAnimTrail/FN_UpdateAnimTrail()
   &ScaleColor(
    Scale Alpha=&Multiply_Float(
     A=&FloatFromCurve(FloatCurve=DataInterface() {propertyValues {Curve {keys [<at 0.2538 0.7462>,<at 1 0>]}}})
     B=$System.FN_Integrator_v2.Attr
    )
    Scale RGB=&MakeVectorFromLinearColorRGB(Color=<rgba 1 0.8915 0.8 1>)
    ScaleA=true
    ScaleRGB=true
   )
   &ScaleRibbonWidth(Ribbon Width Scale=&FloatFromCurve(FloatCurve=DataInterface() {propertyValues {Curve {keys [<at 0 1>,<at 1 1>]}}}))
  ]
  rendererClasses [&NiagaraRibbonRendererProperties]
  renderers [{rendererClass=&NiagaraRibbonRendererProperties}]
 }
 Thrusters {
  emitterUpdateScript [
   &EmitterState()
   &SpawnRate(SpawnRate=48, Use Spawn Probability=false)
  ]
  particleSpawnScript [
   /Niagara/Modules/Spawn/Initialization/InitializeParticle(
    Color=<rgba 0.5 0.5 0.5 0.1>
    Lifetime=0.6
    Module.Write Color=true
    Module.Write Lifetime=true
    Module.Write Mass=false
    Module.Write Material Random=true
    Module.Write Position=false
    Module.Write Scale=false
    Module.Write SpriteRotation=true
    Module.Write SpriteSize=true
    Position=&SimulationPosition()
    Sprite Rotation=&UniformRangedFloat(Minimum=-1, Maximum=1)
    Sprite Size=&Vector2DFromFloat(Value=&Multiply_Float(
      A=&UniformRangedFloat(Minimum=28, Maximum=32)
      B=&FN_FrontEndFloatModifier(Gameplay=1, Front End=0.5)
     ))
   )
   &FN_SetSpawnGroup_v2(Number of Spawn Groups=4)
   &CameraOffset(Camera Offset Amount=2)
  ]
  particleUpdateScript [
   &ParticleState()
   &SkeletalMeshLocation(
    Bone / Socket Index=$Particles.SpawnGroup
    Inherit Velocity=false
    Interpolate Mesh Sampling=false
    MaskSpawnGroup=false
    Mesh Sampling Type=::NewEnumerator5
    Skeletal Mesh=$User.SkeletalMesh
    Socket Sampling Mode=::NewEnumerator5
   )
   /SetVariables(Particles.Position=&Add_Vector(
     A=Linked() {linkedVariable {name=Particles.Position, type=NiagaraPosition}}
     B=&RotateVectorByQuaternion(
      Vector To Rotate=&Multiply_VectorByFloat(
       Vector=&MakeVector(
        X=&Lerp_Float(
         A=&FN_FrontEndFloatModifier(
          Gameplay=15
          Front End=&Add_Float(
           A=10
           B=&FloatFromCurve(FloatCurve=DataInterface() {propertyValues {Curve {keys [<at 3 0>,<at 3.25 12>]}}}, CurveIndex=$System.Age)
          )
         )
         B=&FN_FrontEndFloatModifier(Gameplay=50, Front End=38)
         Alpha=$System.FN_Integrator_v2001.Attr
        )
        Y=0
        Z=0
       )
       Float=$Particles.NormalizedAge
      )
      Quaternion=$Output.SkeletalMeshLocation.SampledRotation
     )
    )) {isSetParameters=true}
   &ScaleColor(
    Scale Alpha=&FloatFromCurve(FloatCurve=DataInterface() {propertyValues {Curve {keys [<at 0 0 RCIM_Cubic>, <at 0.1 1>, {arriveTangent=-1.25, interpMode=RCIM_Cubic, leaveTangent=-1.25, time=0.45, value=1}, <at 0.9 0>]}}})
    ScaleA=true
    ScaleRGB=false
   )
   &ScaleSpriteSize(Scale Factor=&Vector2DFromFloat(Value=&FloatFromCurve(FloatCurve=DataInterface() {propertyValues {Curve {keys [<at 0.5 1>,<at 1 0.4>]}}})))
  ]
  rendererClasses [&NiagaraSpriteRendererProperties]
  renderers [{rendererClass=&NiagaraSpriteRendererProperties}]
 }
 Thrusters_and_Lights {
  emitterUpdateScript [
   &EmitterState(Life Cycle Mode=::NewEnumerator1, Loop Behavior=::NewEnumerator1, Loop Duration=4, UseLoopDelay=false)
   &SpawnBurst_Instantaneous(Spawn Count=4, Use Spawn Probability=false)
  ]
  particleSpawnScript [
   &InitializeParticle(
    Color=<rgba 8 0 0.8612 0.15>
    Mesh Scale=&Multiply_VectorByFloat(Vector=<1.1 0.1 1.1>, Float=&FN_FrontEndFloatModifier(Gameplay=1, Front End=0.385))
    Module.Write Color=true
    Module.Write Lifetime=true
    Module.Write Mass=false
    Module.Write Material Random=true
    Module.Write Position=false
    Module.Write Scale=true
    Module.Write SpriteRotation=true
    Module.Write SpriteSize=true
    Position=&SimulationPosition()
    Sprite Rotation=&UniformRangedFloat(Minimum=0, Maximum=360)
    Sprite Size=&Vector2DFromFloat(Value=5)
   )
   &FN_SetSpawnGroup_v2(Number of Spawn Groups=4)
   /Niagara/Modules/Spawn/Orientation/V2/InitialMeshOrientation(Rotation=<0 0 -0.25>, Use Orientation Vector=true, Use Rotation Vector=true)
  ]
  particleUpdateScript [
   &ParticleState(Kill Particles When Lifetime Has Elapsed=false)
   &SkeletalMeshLocation(
    Apply Sampled Orientation to Mesh Particles=true
    Bone / Socket Index=$Particles.SpawnGroup
    Inherit Velocity=false
    Interpolate Mesh Sampling=false
    MaskSpawnGroup=false
    Mesh Sampling Type=::NewEnumerator5
    Skeletal Mesh=$User.SkeletalMesh
    Socket Sampling Mode=::NewEnumerator5
   )
   &ScaleColor(
    Scale RGB=&Lerp_Vector(
     A=&Multiply_VectorByFloat(
      Vector=&MakeVectorFromLinearColorRGB(Color=&ColorFromCurve(
        ColorCurve=DataInterface() {propertyValues {
          AlphaCurve {keys []}
          BlueCurve {keys [<at 0 1 RCIM_Cubic>,<at 0.2 3 RCIM_Cubic>,<at 0.5 1 RCIM_Cubic>,<at 0.6 3 RCIM_Cubic>,<at 1 1 RCIM_Cubic>]}
          GreenCurve {keys [<at 0 1 RCIM_Cubic>,<at 0.2 3 RCIM_Cubic>,<at 0.5 1 RCIM_Cubic>,<at 0.6 3 RCIM_Cubic>,<at 1 1 RCIM_Cubic>]}
          RedCurve {keys [<at 0 1 RCIM_Cubic>,<at 0.2 3 RCIM_Cubic>,<at 0.5 1 RCIM_Cubic>,<at 0.6 3 RCIM_Cubic>,<at 1 1 RCIM_Cubic>]}
         }}
        CurveIndex=$Emitter.NormalizedLoopAge
       ))
      Float=&FN_FrontEndFloatModifier(Gameplay=0.15, Front End=4)
     )
     B=&Multiply_VectorByFloat(
      Vector=&MakeVectorFromLinearColorRGB(Color=&ColorFromCurve(
        ColorCurve=DataInterface() {propertyValues {
          AlphaCurve {keys []}
          BlueCurve {keys [<at 0 2.5 RCIM_Cubic>,<at 0.2 5 RCIM_Cubic>,<at 0.5 2.5 RCIM_Cubic>,<at 0.6 5 RCIM_Cubic>,<at 1 2.5 RCIM_Cubic>]}
          GreenCurve {keys [<at 0 2.5 RCIM_Cubic>,<at 0.2 5 RCIM_Cubic>,<at 0.5 2.5 RCIM_Cubic>,<at 0.6 5 RCIM_Cubic>,<at 1 2.5 RCIM_Cubic>]}
          RedCurve {keys [<at 0 2.5 RCIM_Cubic>,<at 0.2 5 RCIM_Cubic>,<at 0.5 2.5 RCIM_Cubic>,<at 0.6 5 RCIM_Cubic>,<at 1 2.5 RCIM_Cubic>]}
         }}
        CurveIndex=$Emitter.NormalizedLoopAge
       ))
      Float=1.5
     )
     Alpha=$System.FN_Integrator_v2001.Attr
    )
    ScaleA=false
    ScaleRGB=true
   )
   &FN_Light_v3(
    Create Exponent Attribute=true
    Create Position Attribute=true
    Create Radius Attribute=true
    Exponent=0.5
    Position:Vector3f=&Add_Vector(
     A=Linked() {linkedVariable {name=Particles.Position, type=NiagaraPosition}}
     B=&RotateVectorByQuaternion(
      Vector To Rotate=&MakeVector(X=&FN_FrontEndFloatModifier(Gameplay=15, Front End=6), Y=0, Z=0)
      Quaternion=$Output.SkeletalMeshLocation.SampledRotation
     )
    )
    Radius:NiagaraFloat=&FN_FrontEndFloatModifier(Gameplay=30, Front End=10)
   )
   /SetVariables(Particles.Position=&Add_Vector(
     A=Linked() {linkedVariable {name=Particles.Position, type=NiagaraPosition}}
     B=&RotateVectorByQuaternion(
      Vector To Rotate=&MakeVector(
       X=&FN_FrontEndFloatModifier(Gameplay=&Lerp_Float(A=0, B=4, Alpha=$System.FN_Integrator_v2001.Attr), Front End=0)
       Y=0
       Z=0
      )
      Quaternion=$Output.SkeletalMeshLocation.SampledRotation
     )
    )) {isSetParameters=true}
   &ScaleMeshSize(Scale Factor=&MakeVector(
     X=1
     Y=&Lerp_Float(
      A=&FN_FrontEndFloatModifier(
       Gameplay=0
       Front End=&Multiply_Float(
        A=16
        B=&FloatFromCurve(FloatCurve=DataInterface() {propertyValues {Curve {keys [<at 3 0>,<at 3.25 1>]}}}, CurveIndex=$System.Age)
       )
      )
      B=12
      Alpha=$System.FN_Integrator_v2001.Attr
     )
     Z=1
    ))
  ]
  rendererClasses [&NiagaraLightRendererProperties, &NiagaraMeshRendererProperties]
  renderers [
   {rendererClass=&NiagaraLightRendererProperties, rendererIndex=0}
   {rendererClass=&NiagaraMeshRendererProperties, rendererIndex=1}
  ]
 }
}
summary {
 systemName=NS_Glider_Maverick
 emitters [
  {emitterName=FN_AnimTrail_V4, rendererClasses [&NiagaraRibbonRendererProperties]}
  {emitterName=Thrusters, rendererClasses [&NiagaraSpriteRendererProperties]}
  {
   emitterName=Thrusters_and_Lights
   rendererClasses [&NiagaraLightRendererProperties, &NiagaraMeshRendererProperties]
  }
 ]
 userVariables [
  User.bIsFrontEnd {type=NiagaraBool}
  User.bIsFrontEndPreview {type=NiagaraBool}
  User.bIsFullyDeployed {type=NiagaraBool}
  User.ForwardDot {type=NiagaraFloat, defaultValue=0, description=""}
  User.IsAttacking {type=NiagaraBool}
  User.RibbonIDOffset {type=NiagaraInt32}
  User.RightDot {type=NiagaraFloat, defaultValue=0, description=""}
  User.SkeletalMesh {
   type=NiagaraDataInterfaceSkeletalMesh
   defaultValue {
    struct=NiagaraExt_VariableValue_DataInterface
    value {dataInterface=&self:NiagaraDataInterfaceSkeletalMesh_1}
   }
  }
 ]
}
